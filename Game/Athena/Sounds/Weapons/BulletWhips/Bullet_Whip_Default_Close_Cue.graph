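# Default close-range bullet whip sound cue for Athena, for projectiles whizzing past.
# A Random node picks one of four equally weighted heavy whip waves, then a Modulator varies its pitch and volume on each play for variety.
# It uses the shared bullet whip attenuation asset, the ranged-weapon projectile effects sound class and the bullet whiz-by concurrency rule, with virtualization disabled.
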SoundCue /Game/Athena/Sounds/Weapons/BulletWhips/Bullet_Whip_Default_Close_Cue
root=Modulator_0
nodes [
 {id=Modulator_0, children [Random_0], properties {pitchMax=2, pitchMin=0.8, volumeMax=1, volumeMin=0.5}}
 {
  id=Random_0
  children [WavePlayer_1,WavePlayer_2,WavePlayer_3,WavePlayer_4]
  properties {bShouldExcludeFromBranchCulling=true, weights [1,1,1,1]}
 }
 {id=WavePlayer_1, properties {soundWaveAssetPtr=&/bullet_whip_heavy_01}}
 {id=WavePlayer_2, properties {soundWaveAssetPtr=&/bullet_whip_heavy_02}}
 {id=WavePlayer_3, properties {soundWaveAssetPtr=&/bullet_whip_heavy_03}}
 {id=WavePlayer_4, properties {soundWaveAssetPtr=&/bullet_whip_heavy_04}}
]
properties {
 attenuationSettings=&/bullet_whip_Att.Bullet_Whip_Att
 bBypassVolumeScaleForPriority=true
 maxDistance=2097152
 priority=0.5
 soundClassObject=&Fort_WeaponRanged_Projectile_Effects
 virtualizationMode=Disabled
 volumeMultiplier=0.65
 attenuationOverrides {bAttenuate=false, stereoSpread=0}
 concurrencySet [/Game/Sounds/Concurrency_Rules/Athena_BulletWhizBy_Conc]
}
