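# Sound cue for the close, heavy bullet whip-by effect of the Ferret, in the ranged weapon projectile effects sound class.
# A random node picks one of four equally weighted heavy whip waves, then a modulator varies its pitch and volume.
# Overridden attenuation turns off distance volume falloff and sets a wide stereo spread.
# A whiz-by concurrency rule limits overlapping plays, and virtualization is disabled.
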SoundCue /Game/Athena/Sounds/Weapons/BulletWhips/Bullet_Whip_Ferret_Heavy_Close_Cue
root=Modulator_1
nodes [
 {id=Modulator_1, children [Random_3], properties {pitchMax=2, pitchMin=0.9, volumeMax=1, volumeMin=0.5}}
 {
  id=Random_3
  children [WavePlayer_10, WavePlayer_11, WavePlayer_12, WavePlayer_13]
  properties {bShouldExcludeFromBranchCulling=true, weights [1,1,1,1]}
 }
 {id=WavePlayer_10, properties {soundWaveAssetPtr=&/bullet_whip_heavy_01}}
 {id=WavePlayer_11, properties {soundWaveAssetPtr=&/bullet_whip_heavy_02}}
 {id=WavePlayer_12, properties {soundWaveAssetPtr=&/bullet_whip_heavy_03}}
 {id=WavePlayer_13, properties {soundWaveAssetPtr=&/bullet_whip_heavy_04}}
]
properties {
 bBypassVolumeScaleForPriority=true
 bOverrideAttenuation=true
 maxDistance=2097152
 priority=0.5
 soundClassObject=&Fort_WeaponRanged_Projectile_Effects
 virtualizationMode=Disabled
 volumeMultiplier=0.65
 attenuationOverrides {bAttenuate=false, stereoSpread=5000}
 concurrencySet [/Game/Sounds/Concurrency_Rules/Athena_BulletWhizBy_Conc]
}
