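# Sound cue for the close, standard Ferret bullet whip, a projectile whiz-by effect in the Athena weapon sounds.
# A Random node picks one of four equally weighted whip waves, and a Modulator above it randomizes pitch and volume for variation.
# An attenuation override turns off distance volume falloff, while a whiz-by concurrency rule governs overlapping instances and virtualization is disabled.
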SoundCue /Game/Athena/Sounds/Weapons/BulletWhips/Bullet_Whip_Ferret_Standard_Close_Cue
root=Modulator_2
nodes [
 {id=Modulator_2, children [Random_4], properties {pitchMax=1.5, volumeMax=1, volumeMin=0.5}}
 {
  id=Random_4
  children [WavePlayer_14, WavePlayer_15, WavePlayer_16, WavePlayer_17]
  properties {bShouldExcludeFromBranchCulling=true, weights [1,1,1,1]}
 }
 {id=WavePlayer_14, properties {soundWaveAssetPtr=&/bullet_whip_01}}
 {id=WavePlayer_15, properties {soundWaveAssetPtr=&/bullet_whip_02}}
 {id=WavePlayer_16, properties {soundWaveAssetPtr=&/bullet_whip_03}}
 {id=WavePlayer_17, properties {soundWaveAssetPtr=&/bullet_whip_04}}
]
properties {
 bBypassVolumeScaleForPriority=true
 bOverrideAttenuation=true
 maxDistance=2097152
 priority=0.5
 soundClassObject=&Fort_WeaponRanged_Projectile_Effects
 virtualizationMode=Disabled
 volumeMultiplier=1.4
 attenuationOverrides {bAttenuate=false, stereoSpread=5000}
 concurrencySet [/Game/Sounds/Concurrency_Rules/Athena_BulletWhizBy_Conc]
}
