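# Sound cue for the close-range whip of a passing sniper bullet in Athena.
# A Random node picks one of six equally weighted wave variants, and a root Modulator varies its volume and raises its pitch by a random amount.
# It uses the bullet whip attenuation, the ranged projectile effects sound class and a bullet whiz-by concurrency rule, with virtualization disabled.
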SoundCue /Game/Athena/Sounds/Weapons/BulletWhips/Bullet_Whip_Sniper_Close_Cue
root=Modulator_1
nodes [
 {id=Modulator_1, children [Random_2], properties {pitchMax=1.5, pitchMin=1, volumeMax=1, volumeMin=0.8}}
 {
  id=Random_2
  children [WavePlayer_6, WavePlayer_7, WavePlayer_8, WavePlayer_9, WavePlayer_10, WavePlayer_11]
  properties {bShouldExcludeFromBranchCulling=true, weights [1,1,1,1,1,1]}
 }
 {id=WavePlayer_10, properties {soundWaveAssetPtr=&/Bullet_Whip_Sniper_Close_05}}
 {id=WavePlayer_11, properties {soundWaveAssetPtr=&/Bullet_Whip_Sniper_Close_06}}
 {id=WavePlayer_6, properties {soundWaveAssetPtr=&/Bullet_Whip_Sniper_Close_01}}
 {id=WavePlayer_7, properties {soundWaveAssetPtr=&/Bullet_Whip_Sniper_Close_02}}
 {id=WavePlayer_8, properties {soundWaveAssetPtr=&/Bullet_Whip_Sniper_Close_03}}
 {id=WavePlayer_9, properties {soundWaveAssetPtr=&/Bullet_Whip_Sniper_Close_04}}
]
properties {
 attenuationSettings=&/bullet_whip_Att.Bullet_Whip_Att
 bBypassVolumeScaleForPriority=true
 maxDistance=2097152
 priority=0.5
 soundClassObject=&Fort_WeaponRanged_Projectile_Effects
 virtualizationMode=Disabled
 volumeMultiplier=0.5
 attenuationOverrides {bAttenuate=false, stereoSpread=0}
 concurrencySet [/Game/Sounds/Concurrency_Rules/Athena_BulletWhizBy_Conc]
}
