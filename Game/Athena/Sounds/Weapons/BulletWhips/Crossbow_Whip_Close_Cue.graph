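# Sound cue for a close crossbow projectile whiz-by, part of the Athena bullet whip weapon sounds.
# A random node picks among four equally weighted crossbow whip waves and is excluded from branch culling.
# It plays quietly through the ranged projectile effects sound class, with overridden attenuation that disables distance falloff.
# A shared bullet whiz-by concurrency rule governs simultaneous plays, and virtualization is disabled.
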SoundCue /Game/Athena/Sounds/Weapons/BulletWhips/Crossbow_Whip_Close_Cue
root=Random_1
nodes [
 {
  id=Random_1
  children [WavePlayer_0,WavePlayer_5,WavePlayer_6,WavePlayer_7]
  properties {bShouldExcludeFromBranchCulling=true, weights [1,1,1,1]}
 }
 {id=WavePlayer_0, properties {soundWaveAssetPtr=&/Crossbow_Whip_02}}
 {id=WavePlayer_5, properties {soundWaveAssetPtr=&/Crossbow_Whip_03}}
 {id=WavePlayer_6, properties {soundWaveAssetPtr=&/Crossbow_Whip_04}}
 {id=WavePlayer_7, properties {soundWaveAssetPtr=&/Crossbow_Whip_05}}
]
properties {
 bBypassVolumeScaleForPriority=true
 bOverrideAttenuation=true
 maxDistance=2097152
 priority=0.5
 soundClassObject=&Fort_WeaponRanged_Projectile_Effects
 virtualizationMode=Disabled
 volumeMultiplier=0.1
 attenuationOverrides {bAttenuate=false, stereoSpread=5000}
 concurrencySet [/Game/Sounds/Concurrency_Rules/Athena_BulletWhizBy_Conc]
}
