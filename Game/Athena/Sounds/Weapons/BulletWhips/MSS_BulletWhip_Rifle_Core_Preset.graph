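# MetaSound Source preset for rifle bullet whip (whiz-by) one-shots, inheriting its graph from MSS_BulletWhip_Default_Core.
# It supplies eleven rifle whip waves through the Sounds array.
# The Fort.BulletWhip interface exposes pass-by distance, azimuth, alignment, crossfade, curve power, impact-distance and stop inputs to the parent graph.
# Output is quad, routed to the ranged-weapon projectile effects sound class under bullet whiz-by concurrency.
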MetaSoundSource /Game/Athena/Sounds/Weapons/BulletWhips/MSS_BulletWhip_Rifle_Core_Preset
preset_parent=&/MSS_BulletWhip_Default_Core
inputs [
 BulletWhip.IgnoreImpactDistance {type=Float}
 BulletWhip.OnStop {type=Trigger}
 BulletWhip.PassByAlignment {type=Float}
 BulletWhip.PassByAzimuth {type=Float}
 BulletWhip.PassByCrossfadeTime {type=Time}
 BulletWhip.PassByCurvePower {type=Float}
 BulletWhip.PassByDistance {type=Float}
 Sounds {type=WaveAsset:Array, access=Constructor}
 UE.Source.OnPlay {type=Trigger}
]
interfaces [Fort.BulletWhip:1.0, UE.OutputFormat.Quad:1.0, UE.Source.OneShot:1.0, UE.Source:1.1]
outputs [
 "UE.OutputFormat.Quad.Audio:0" {type=Audio}
 "UE.OutputFormat.Quad.Audio:1" {type=Audio}
 "UE.OutputFormat.Quad.Audio:4" {type=Audio}
 "UE.OutputFormat.Quad.Audio:5" {type=Audio}
 UE.Source.OneShot.OnFinished {type=Trigger}
]
pages [Default {input_defaults {
   BulletWhip.IgnoreImpactDistance=0.1
   BulletWhip.OnStop=null
   BulletWhip.PassByAlignment=0
   BulletWhip.PassByAzimuth=-90
   BulletWhip.PassByCrossfadeTime=0.2
   BulletWhip.PassByCurvePower=4
   BulletWhip.PassByDistance=1
   UE.Source.OnPlay=null
   Sounds [
    /Game/Sounds/Fort_Weapons/Whips/Bullet_Whip_Rifle_01
    /Game/Sounds/Fort_Weapons/Whips/Bullet_Whip_Rifle_02
    /Game/Sounds/Fort_Weapons/Whips/Bullet_Whip_Rifle_03
    /Game/Sounds/Fort_Weapons/Whips/Bullet_Whip_Rifle_04
    /Game/Sounds/Fort_Weapons/Whips/Bullet_Whip_Rifle_05
    /Game/Sounds/Fort_Weapons/Whips/Bullet_Whip_Rifle_06
    /Game/Sounds/Fort_Weapons/Whips/Bullet_Whip_Rifle_07
    /Game/Sounds/Fort_Weapons/Whips/Bullet_Whip_Rifle_08
    /Game/Sounds/Fort_Weapons/Whips/Bullet_Whip_Rifle_09
    /Game/Sounds/Fort_Weapons/Whips/Bullet_Whip_Rifle_10
    /Game/Sounds/Fort_Weapons/Whips/Bullet_Whip_Rifle_11
   ]
  }}]
properties {
 assetImportData=&self:AssetImportData
 numChannels=4
 outputFormat=Quad
 sampleRate=48000
 soundClassObject=&Fort_WeaponRanged_Projectile_Effects
 volume=0.9
 concurrencySet [/Game/Sounds/Concurrency_Rules/Athena_BulletWhizBy_Conc]
 modulationSettings {volumeRouting=Union}
}
